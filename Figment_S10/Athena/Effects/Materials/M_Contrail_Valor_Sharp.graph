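# Translucent, two-sided unlit contrail material for beam trails and particle sprites.
# Color multiplies time-of-day tint, a gradient texture or UV mask, and particle color, then boosts it through HDR remapping or a constant.
# Opacity adds tiling, optionally panned smoke noise to a UV gradient, shapes it with edge and optional masks, and depth-fades outside mobile.
# A Flip scalar blends in a rotated color map, inverted gradient and offset noise.
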Material /Figment_S10/Athena/Effects/Materials/M_Contrail_Valor_Sharp
graph {
 properties {
  bUsedWithBeamTrails=true
  bUsedWithParticleSprites=true
  blendMode=BLEND_Translucent
  pixelDepthOffsetMode=PDOM_Legacy
  shadingModel=MSM_Unlit
  translucencyLightingMode=TLM_VolumetricPerVertexNonDirectional
  twoSided=true
  usedShadingModels=Unlit
 }
 nodes [
  Constant() {r=2}
  n1 = VectorParameter() {parameterName=BaseColor, defaultValue=<rgba 1 1 1 0>}
  n2 = VectorParameter() {parameterName=NightColor, defaultValue=<rgba 0.825 1.0472 1.2 1>}
  n3 = MaterialFunctionCall(Float2=TextureCoordinate()) {materialFunction=&BreakOutFloat2Components}
  n4 = Multiply(A=Sine(None=Multiply(A=@n3.G) {constB=0.5}), B=OneMinus(None=@n3.R))
  n5 = StaticSwitchParameter(
   True=@n4
   False=TextureSampleParameter2D() {parameterName=TextureSample, texture=/Game/Effects/Fort_Effects/Textures/Grad/T_GliderColorMap}.RGB
  ) {parameterName=UseTexCoordAsMask}
  n6 = TextureCoordinate()
  n7 = ScalarParameter() {defaultValue=0.5, parameterName=Rot}
  n8 = StaticSwitchParameter(
   True=@n4
   False=TextureSampleParameter2D(UVs=MaterialFunctionCall(UVs=@n6, "Rotation Angle (0-1)"=@n7) {materialFunction=&CustomRotator}.Rotated Values) {parameterName=TextureSample, texture=&T_GliderColorMap}.RGB
  ) {parameterName=UseTexCoordAsMask}
  n9 = ScalarParameter() {parameterName=Flip}
  n10 = ParticleColor()
  n11 = Multiply(
   A=Multiply(
    A=MaterialFunctionCall(Color_Morning=@n1.RGB, Color_Day=@n1.RGB, Color_Evening=@n1.RGB, Color_Night=@n2.RGB) {materialFunction=&Color_Time_of_Day}.Result
    B=LinearInterpolate(A=@n5, B=@n8, Alpha=@n9)
   )
   B=@n10.RGB
  )
  n12 = Multiply(A=@n11) {constB=2.5}
  n13 = StaticSwitchParameter(
   True=@n12
   False=MaterialFunctionCall(LDR Input=Clamp(None=@n11), HDR Dynamic Range=ScalarParameter() {defaultValue=50, parameterName=HDRPower}, HDR Tint=@n11) {materialFunction=&DeriveHDRfromLDR}.RemappedRange
  ) {desc=Add no use HDR, parameterName=No use HDR}
  FeatureLevelSwitch(Default=@n13, ES2=@n12, ES3_1=@n12, SM4_REMOVED=@n13)
  n14 = TextureCoordinate()
  n15 = ComponentMask(None=TextureCoordinate()) {g=true}
  n16 = OneMinus(None=@n15)
  n17 = Multiply(
   A=TextureCoordinate() {coordinateIndex=1}
   B=AppendVector(A=ScalarParameter() {defaultValue=1, parameterName=NoiseUtliing}, B=ScalarParameter() {defaultValue=0.5, parameterName=NoiseVtliing})
  )
  n18 = ScalarParameter() {defaultValue=0.5, parameterName=PanSpeedU}
  n19 = ScalarParameter() {parameterName=PanSpeedV}
  n20 = StaticSwitchParameter(True=Panner(Coordinate=@n17, Speed=AppendVector(A=@n18, B=@n19)) {speedX=0.5}, False=@n17) {group=FX, parameterName=UsePanner?}
  n21 = ScalarParameter() {defaultValue=1, desc=Add Mask texture intensity, parameterName=Noise_Intensity}
  n22 = Multiply(
   A=TextureSampleParameter2D(UVs=@n20) {parameterName=Noise, texture=/Game/Effects/Fort_Effects/Textures/Smoke/T_SmokeTile_Strip_03_Height_TILE}.R
   B=@n21
  )
  n23 = TextureSample(UVs=@n20) {samplerType=SAMPLERTYPE_Grayscale, texture=&T_Smoke_Trail_Soft}
  n24 = AppendVector(A=ComponentMask(None=@n20) {r=true}, B=Add(A=ComponentMask(None=@n20) {g=true}) {constB=0.5})
  n25 = Multiply(
   A=TextureSampleParameter2D(UVs=@n24) {parameterName=Noise, texture=&T_SmokeTile_Strip_03_Height_TILE}.R
   B=@n21
  )
  n26 = TextureSample(UVs=@n24) {samplerType=SAMPLERTYPE_Grayscale, texture=&T_Smoke_Trail_Soft}
  n27 = Add(
   A=LinearInterpolate(A=Power(Base=@n15), B=Power(Base=@n16), Alpha=@n9)
   B=LinearInterpolate(
    A=StaticSwitchParameter(True=@n22, False=@n23.G) {group=FX, parameterName=UseTextureParam?}
    B=StaticSwitchParameter(True=@n25, False=@n26.G) {group=FX, parameterName=UseTextureParam?}
    Alpha=@n9
   )
  )
  n28 = ComponentMask(None=TextureCoordinate()) {r=true}
  n29 = Multiply(A=OneMinus(None=@n28), B=Power(Base=@n28, Exp=ScalarParameter() {defaultValue=2, parameterName=MaskPow}) {constExponent=0.6})
  n30 = Multiply(A=@n10.A, B=Clamp(None=Multiply(A=@n27, B=@n29)))
  n31 = DepthFade(Opacity=@n30, FadeDistance=ScalarParameter() {defaultValue=30, parameterName=DepthFade})
  n32 = FeatureLevelSwitch(Default=@n31, ES2=@n30, ES3_1=@n30, SM4_REMOVED=@n31)
  n33 = StaticSwitchParameter(True=Multiply(A=ComponentMask(None=@n14) {g=true}, B=@n32), False=@n32) {parameterName=RemoveHardEdge}
  n34 = Multiply(A=MaterialFunctionCall(Float2=TextureCoordinate()) {materialFunction=&BreakOutFloat2Components}.R)
  n35 = ScalarParameter() {defaultValue=2, parameterName=Mask_Power}
  n36 = StaticSwitchParameter(True=Multiply(A=@n33, B=Power(Base=@n34, Exp=@n35)), False=@n33) {parameterName=Add Mask}
  n37 = SphereMask(A=TextureCoordinate(), B=Constant2Vector() {g=0.5, r=1}) {attenuationRadius=1, hardnessPercent=50}
  n38 = ScalarParameter() {defaultValue=2, parameterName=Mask_Power}
  StaticSwitchParameter(True=Multiply(A=@n36, B=Power(Base=@n37, Exp=@n38)), False=@n36) {parameterName=Add Sphere Mask}
 ]
}
